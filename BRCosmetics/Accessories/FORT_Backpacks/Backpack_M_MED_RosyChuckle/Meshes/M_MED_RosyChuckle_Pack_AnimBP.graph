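# Animation Blueprint for the RosyChuckle backpack cosmetic, posing its Body bone and adding secondary motion to its bells.
# The event graph maps the yaw of the owner mesh's spine_05 delta from ref pose, relative to spine_01, into a Parashuting value.
# A constant additive Body offset feeds a second one weighted by Parashuting.
# Four chained AnimDynamics nodes then simulate the center, left and right bell bones with angular springs and limits.
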AnimBlueprint /BRCosmetics/Accessories/FORT_Backpacks/Backpack_M_MED_RosyChuckle/Meshes/M_MED_RosyChuckle_Pack_AnimBP
target_skeleton=&/M_MED_RosyChuckle_Pack_Skeleton
graphs [
 AnimGraph {nodes [
   n1 = "Transform (Modify) Bone - Bone: Body"(ComponentPose=Local To Component(LocalPose=Input Pose()), Alpha=1, Translation="0,0,6", Rotation="0.000,0.000,-10.000", Scale="1,1,1") {
    rotationMode=BMM_Additive
    scaleMode=BMM_Additive
    translationMode=BMM_Additive
    alphaBoolBlend {blendOption=Linear}
    boneToModify {boneName=Body}
    rotation {roll=-10}
    translation {z=6}
   }
   n2 = "Transform (Modify) Bone - Bone: Body"(ComponentPose=@n1, Alpha=Get Parashuting() -> (Parashuting:double), Translation="0,-4,1", Rotation="0,0,40", Scale="1,1,1") {
    rotationMode=BMM_Additive
    scaleMode=BMM_Additive
    translationMode=BMM_Additive
    alphaBoolBlend {blendOption=Linear}
    boneToModify {boneName=Body}
    rotation {roll=40}
    translation {y=-4, z=1}
   }
   n3 = Anim Dynamics - Bone: Dyn_bell_C(ComponentPose=@n2, Alpha=1, ExternalForce="0,0,0") {
    angularDampingOverride=0.999
    angularSpringConstant=700
    bAngularSpring=true
    bOverrideAngularDamping=true
    bUseGravityOverride=true
    simulationSpace=RootRelative
    boundBone {boneName=Dyn_bell_C}
    componentAppliedLinearAccClamp=<1000 1000 1000>
    componentLinearAccScale=<0.8 0.8 0.8>
    componentLinearVelScale=<0.8 0.8 0.8>
    gravityOverride {z=-980}
    physicsBodyDefinitions [{
      boundBone {boneName=Dyn_bell_C}
      boxExtents=<10 10 10>
      constraintSetup {angularLimitsMax=<0 0 0>, angularLimitsMin=<0 -50 0>, angularTarget=<1 0 0>}
      localJointOffset=<0.8 0 0>
     }]
   }
   n4 = Anim Dynamics - Bone: Dyn_bell_C_01(ComponentPose=@n3, Alpha=1, ExternalForce="0,0,0") {
    angularDampingOverride=0.999
    angularSpringConstant=700
    bAngularSpring=true
    bOverrideAngularDamping=true
    bUseGravityOverride=true
    simulationSpace=RootRelative
    boundBone {boneName=Dyn_bell_C_01}
    componentAppliedLinearAccClamp=<1000 1000 1000>
    componentLinearAccScale=<0.8 0.8 0.8>
    componentLinearVelScale=<0.8 0.8 0.8>
    gravityOverride {z=-980}
    physicsBodyDefinitions [{
      boundBone {boneName=Dyn_bell_C_01}
      boxExtents=<10 10 10>
      constraintSetup {angularLimitsMax=<0 0 20>, angularLimitsMin=<0 0 -20>, angularTarget=<1 0 0>}
      localJointOffset=<0.8 0 0>
     }]
   }
   n5 = Anim Dynamics - Bone: Dyn_bell_L(ComponentPose=@n4, Alpha=1, ExternalForce="0,0,0") {
    angularDampingOverride=0.999
    angularSpringConstant=700
    bAngularSpring=true
    bOverrideAngularDamping=true
    bUseGravityOverride=true
    simulationSpace=RootRelative
    boundBone {boneName=Dyn_bell_L}
    componentAppliedLinearAccClamp=<1100 1100 1100>
    componentLinearAccScale=<0.9 0.9 0.9>
    componentLinearVelScale=<0.9 0.9 0.9>
    gravityOverride {z=-980}
    physicsBodyDefinitions [{
      boundBone {boneName=Dyn_bell_L}
      boxExtents=<11 11 11>
      constraintSetup {angularLimitsMax=<0 0 15>, angularLimitsMin=<0 -25 0>, angularTarget=<1 0 0>}
      localJointOffset=<0.6 0 0>
     }]
   }
   n6 = Anim Dynamics - Bone: Dyn_bell_R(ComponentPose=@n5, Alpha=1, ExternalForce="0,0,0") {
    angularDampingOverride=0.999
    angularSpringConstant=700
    bAngularSpring=true
    bOverrideAngularDamping=true
    bUseGravityOverride=true
    simulationSpace=RootRelative
    boundBone {boneName=Dyn_bell_R}
    componentAppliedLinearAccClamp=<1100 1100 1100>
    componentLinearAccScale=<0.9 0.9 0.9>
    componentLinearVelScale=<0.9 0.9 0.9>
    gravityOverride {z=-980}
    physicsBodyDefinitions [{
      boundBone {boneName=Dyn_bell_R}
      boxExtents=<10 10 10>
      constraintSetup {angularLimitsMax=<0 0 15>, angularLimitsMin=<0 -30 0>, angularTarget=<1 0 0>}
      localJointOffset=<0.9 0 0>
     }]
   }
   Output Pose(Result=Component To Local(ComponentPose=@n6)) {layerGroup=None}
  ]}
 EventGraph {nodes [
   Event BlueprintUpdateAnimation()
   n1 = GetDeltaTransformFromRefPose!(self=GetOwnerSkeletalMeshComponent(), BoneName=spine_05, BaseName=spine_01) -> (ReturnValue_Location:Vector, ReturnValue_Rotation:Rotator, ReturnValue_Scale:Vector)
   > Set Parashuting(Parashuting:double=MapRangeClamped(Value=BreakRotator(InRot=@n1.ReturnValue_Rotation).Yaw, InRangeA=-15, InRangeB=-50)) -> (Output_Get:double)
   TryGetPawnOwner()
  ]}
]
variables [Parashuting {type=number}]
